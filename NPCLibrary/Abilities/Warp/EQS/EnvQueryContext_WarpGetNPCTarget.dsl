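; EQS context for the NPC warp ability that supplies the querying NPC's current target as a single actor.
; ProvideSingleActor casts the querier to a FortPlayerPawn, gets its controller's blackboard and reads the object stored under the AI_Global_Target key.
; It returns that object as an Actor, or nothing when either cast fails or the blackboard is invalid.
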(fn ProvideSingleActor
  (params (QuerierObject :type Object%20Reference) (QuerierActor :type Actor%20Object%20Reference))
  (bind _asfort_player_pawn (Utilities|Casting|CastToFortPlayerPawn QuerierActor)
    (:then
      (Utilities|IsValid (AI|GetBlackboard (Pawn|GetController _asfort_player_pawn))
        (:Is%20Valid
          (bind _asactor (Utilities|Casting|CastToActor (AI|Components|Blackboard|GetValueasObject (AI|GetBlackboard (Pawn|GetController _asfort_player_pawn)) (LocalVariables|GetAI_Global_Target)))
            (:then
              (return _asactor))
            (:CastFailed
              (return 0))))
        (:Is%20Not%20Valid
          (return 0))))
    (:CastFailed
      (return 0))))
