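# EQS query for the Warp NPC ability that finds candidate teleport spots around the NPC's warp target.
# An OnCircle generator places a ring of points around the target context and projects them to navigation within a limited vertical range.
# A pathfinding test against the target uses the any-traversable NPC nav filter.
# A 2D distance test filters out points too close to the querier, so the warp covers a minimum distance.
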EnvQuery /NPCLibrary/Abilities/Warp/EQS/Warp_EQS_FindTeleportAroundTarget
options [{
  generator {
   name=EnvQueryGenerator_OnCircle_0
   class=EnvQueryGenerator_OnCircle
   properties {
    circleCenter=&/EnvQueryContext_WarpGetNPCTarget.EnvQueryContext_WarpGetNPCTarget_C
    circleRadius=800
    numberOfPoints=8
    projectionData {
     extentX=0
     navigationFilter=&RecastFilter_UseDefaultArea
     postProjectionVerticalOffset=0
     projectDown=150
     projectUp=150
    }
    traceData {navigationFilter=null, traceMode=None}
   }
  }
  tests [
   EnvQueryTest_Pathfinding_2 {
    class=EnvQueryTest_Pathfinding
    properties {
     context=&EnvQueryContext_WarpGetNPCTarget_C
     filterClass=&NavFilter_NPC_AnyTraversable_C
    }
   }
   EnvQueryTest_Distance_0 {
    class=EnvQueryTest_Distance
    properties {
     distanceTo=&EnvQueryContext_Querier
     filterType=Minimum
     floatValueMin=400
     scoringEquation=Linear
     scoringFactor=1
     testComment=""
     testMode=Distance2D
     testPurpose=Filter
    }
   }
  ]
 }]
