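# EQS query that finds valid teleport destinations for the NPC Warp ability.
# An OnCircle generator places a small ring of candidate points at a fixed radius, with tracing from the circle center disabled and limited vertical projection.
# A single pathfinding test evaluates those points using the NPC any-traversable navigation filter.
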EnvQuery /NPCLibrary/Abilities/Warp/EQS/Warp_EQS_FindValidTeleportLocation
options [{
  generator {
   name=EnvQueryGenerator_OnCircle_0
   class=EnvQueryGenerator_OnCircle
   properties {
    circleRadius=250
    numberOfPoints=6
    projectionData {
     extentX=0
     navigationFilter=&RecastFilter_UseDefaultArea
     postProjectionVerticalOffset=0
     projectDown=150
     projectUp=80
    }
    traceData {navigationFilter=null, traceMode=None}
   }
  }
  tests [EnvQueryTest_Pathfinding_2 {
    class=EnvQueryTest_Pathfinding
    properties {filterClass=&NavFilter_NPC_AnyTraversable_C}
   }]
 }]
